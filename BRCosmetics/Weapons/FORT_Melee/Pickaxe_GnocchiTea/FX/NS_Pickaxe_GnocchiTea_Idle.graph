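# Niagara idle effect for the GnocchiTea pickaxe, built from a single FanBlur sprite emitter that, by its name, shows a motion blur on a fan.
# A burst-spawned sprite follows a skeletal mesh socket and takes its facing and alignment from that socket's rotation.
# A system Timeline value drives curves that scale the sprite in and out and fade its alpha, while a system time-of-day color tints it.
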NiagaraSystem /BRCosmetics/Weapons/FORT_Melee/Pickaxe_GnocchiTea/FX/NS_Pickaxe_GnocchiTea_Idle
dependencies {
 usedDataInterfaces [&NiagaraDataInterfaceCurve, &NiagaraDataInterfaceSkeletalMesh, &NiagaraDataInterfaceColorCurve]
 usedDynamicInputs [
  &ColorFromCurve
  &FNTimeOfDayScalar
  &SetBoolByIntComparison
 ]
 usedModules [
  &self:SystemSpawnScript.SystemScriptSource.SystemScriptGraph.NiagaraNodeAssignment_0.SetVariables_#1
  &self:SystemSpawnScript.SystemScriptSource.SystemScriptGraph.NiagaraNodeAssignment_2.SetVariables_#2
  &SystemState
  &self:SystemSpawnScript.SystemScriptSource.SystemScriptGraph.NiagaraNodeAssignment_3.SetVariables_#3
  &self:SystemSpawnScript.SystemScriptSource.SystemScriptGraph.NiagaraNodeAssignment_4.SetVariables_#4
  &Timeline
 ]
}
emitters {FanBlur {
  emitterUpdateScript [
   &EmitterState(
    Inactive Response=::NewEnumerator1
    Life Cycle Mode=::NewEnumerator1
    Loop Behavior=::NewEnumerator1
    Loop Duration Mode=::NewEnumerator1
    UseLoopDelay=false
   )
   &SpawnBurst_Instantaneous(Use Spawn Probability=false, UseLoopCountLimit=false)
  ]
  particleSpawnScript [&InitializeParticle(
    Color=<rgba 0.3385 0.3385 0.3385 1>
    Sprite Size Mode=::NewEnumerator3
    Uniform Sprite Size=61
    UsePositionOffset=false
    Write Material Random=true
   )]
  particleUpdateScript [
   &ParticleState(Kill Particles When Lifetime Has Elapsed=false, Let Infinitely Lived Particles Die When Emitter Deactivates=true)
   &ScaleSpriteSize(
    Uniform Curve Index=$System.Timeline.Time
    Uniform Curve Sprite Scale=DataInterface() {propertyValues {Curve {keys [{tangentMode=RCTM_User, time=0, value=0}, <at 0.03 1>, <at 0.97 1>, {tangentMode=RCTM_User, time=1, value=0}]}}}
   )
   &SkeletalMeshLocation(
    Bone / Socket Index=2
    MaskSpawnGroup=false
    Mesh Sampling Type=::NewEnumerator5
    Skeletal Mesh=$System.Skeletal Mesh
    Socket Sampling Mode=::NewEnumerator5
   )
   &SpriteFacingAndAlignment(
    Sprite Alignment=&RotateVectorByQuaternion(Vector To Rotate=<0 0 1>, Quaternion=$Particles.SkeletalMeshLocation.SampledRotation)
    Sprite Facing=&RotateVectorByQuaternion(Vector To Rotate=<-1 0 0>, Quaternion=$Particles.SkeletalMeshLocation.SampledRotation)
    WriteAlignment=true
    WriteFacing=true
   ) {module=SpriteFacingAndAlignment001}
   &ScaleColor(
    Scale Alpha=&FloatFromCurve(
     FloatCurve=DataInterface() {propertyValues {Curve {keys [
         <at 0 0 RCIM_Cubic>
         {arriveTangent=0.143, interpMode=RCIM_Cubic, leaveTangent=0.143, tangentMode=RCTM_User, time=0.0231, value=1}
         {arriveTangent=0.0742, interpMode=RCIM_Cubic, leaveTangent=0.0742, tangentMode=RCTM_User, time=0.3391, value=1}
         <at 1 0>
        ]}}}
     CurveIndex=$System.Timeline.Time
     Scale Curve=0.35
    )
    Scale RGB=&Multiply_Vector(
     A=&MakeVectorFromLinearColorRGB(Color=$System.TODColor)
     B=&VectorFromFloat(Value=1)
    )
    ScaleA=true
    ScaleRGB=true
   )
  ]
  rendererClasses [&NiagaraSpriteRendererProperties]
  renderers [{rendererClass=&NiagaraSpriteRendererProperties}]
 }}
summary {
 systemName=NS_Pickaxe_GnocchiTea_Idle
 emitters [{emitterName=FanBlur, rendererClasses [&NiagaraSpriteRendererProperties]}]
 userVariables [
  User.bIsFrontEnd {type=NiagaraBool}
  User.bIsFrontEndPreview {type=NiagaraBool}
  User.RibbonIDOffset {type=NiagaraInt32}
  User.SwingCount {type=NiagaraInt32}
 ]
}
