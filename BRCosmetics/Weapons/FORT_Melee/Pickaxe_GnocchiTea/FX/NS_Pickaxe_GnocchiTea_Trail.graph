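# Niagara system for the swing trail of the GnocchiTea pickaxe, drawn as three anim-trail ribbons named Blades, Body and Body2.
# Each emitter follows its own bone or socket pair on the system's skeletal mesh and spawns short-lived, tinted particles.
# Time-of-day color and float values scale the tint, while alpha follows a swing fade curve.
# Blades also narrows over life, and Body is offset toward the camera.
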NiagaraSystem /BRCosmetics/Weapons/FORT_Melee/Pickaxe_GnocchiTea/FX/NS_Pickaxe_GnocchiTea_Trail
dependencies {
 usedDataInterfaces [&NiagaraDataInterfaceCurve, &NiagaraDataInterfaceSkeletalMesh, &NiagaraDataInterfaceColorCurve]
 usedDynamicInputs [&ColorFromCurve, &FNTimeOfDayScalar]
 usedModules [
  &self:SystemSpawnScript.SystemScriptSource.SystemScriptGraph.NiagaraNodeAssignment_0.SetVariables_#1
  &SystemState
  &self:SystemSpawnScript.SystemScriptSource.SystemScriptGraph.NiagaraNodeAssignment_1.SetVariables_#2
  &self:SystemSpawnScript.SystemScriptSource.SystemScriptGraph.NiagaraNodeAssignment_3.SetVariables_#3
 ]
}
emitters {
 Blades {
  emitterSpawnScript [&FN_RibbonIDOffsetUtility()]
  emitterUpdateScript [
   &EmitterState()
   &FN_SpawnAnimTrail()
  ]
  particleSpawnScript [
   &FN_InitializeAnimTrailBoneIndecies(First Bone or Socket Index=2, Second Bone or Socket Index=3)
   &FN_InitializeAnimTrail(
    Color=&MultiplyLinearColor(Linear Color=<rgba 0.5 0.5 0.5 1>, Scale Factor=&MakeVector4_FromLinearColor(LinearColor=$System.TODColor))
    Lifetime=0.14
    SkeletalMesh=$System.Skeletal Mesh
   )
  ]
  particleUpdateScript [
   &FN_UpdateAnimTrail(SkeletalMesh=$System.Skeletal Mesh)
   &ScaleColor(
    Scale Alpha=&Multiply_Float(
     A=&FloatFromCurve(
      FloatCurve=DataInterface() {propertyValues {Curve {keys [<at 0 1.4 RCIM_Cubic>, {arriveTangent=-2.3333, interpMode=RCIM_Cubic, leaveTangent=-2.3333, time=0.0879, value=1}, <at 0.6 0>]}}}
      Scale Curve=$System.SwingGeneralFade
     )
     B=0.9
    )
    Scale RGB=<0.8 0.8 0.8>
    ScaleA=true
    ScaleRGB=true
   )
   &ScaleColor(Scale RGB=&VectorFromFloat(Value=$System.TODFloat), ScaleA=false, ScaleRGB=true) {module=ScaleColor001}
   &ScaleRibbonWidth(Ribbon Width Scale=&FloatFromCurve(FloatCurve=DataInterface() {propertyValues {Curve {keys [<at 0.1 1 RCIM_Cubic>,<at 1 0>]}}}))
  ]
  rendererClasses [&NiagaraRibbonRendererProperties]
  renderers [{rendererClass=&NiagaraRibbonRendererProperties}]
 }
 Body {
  emitterSpawnScript [&FN_RibbonIDOffsetUtility()]
  emitterUpdateScript [
   &EmitterState()
   &FN_SpawnAnimTrail()
  ]
  particleSpawnScript [
   &FN_InitializeAnimTrailBoneIndecies()
   &FN_InitializeAnimTrail(
    Color=&MultiplyLinearColor(
     Linear Color=<rgba 0.276 0.0699 0.0546 1>
     Scale Factor=&MakeVector4_FromLinearColor(LinearColor=$System.TODColor)
    )
    Lifetime=0.14
    SkeletalMesh=$System.Skeletal Mesh
   )
   &CameraOffset(Camera Offset Amount=6)
  ]
  particleUpdateScript [
   &FN_UpdateAnimTrail(SkeletalMesh=$System.Skeletal Mesh)
   &ScaleColor(
    Scale Alpha=&Multiply_Float(
     A=&FloatFromCurve(
      FloatCurve=DataInterface() {propertyValues {Curve {keys [<at 0 1.4 RCIM_Cubic>, {interpMode=RCIM_Cubic, leaveTangent=-2.522, tangentMode=RCTM_Break, time=0.0879, value=1}, <at 0.6 0>]}}}
      Scale Curve=$System.SwingGeneralFade
     )
     B=1.2
    )
    Scale RGB=<0.8 0.8 0.8>
    ScaleA=true
    ScaleRGB=true
   )
   &ScaleColor(Scale RGB=&VectorFromFloat(Value=$System.TODFloat), ScaleA=false, ScaleRGB=true) {module=ScaleColor001}
  ]
  rendererClasses [&NiagaraRibbonRendererProperties]
  renderers [{rendererClass=&NiagaraRibbonRendererProperties}]
 }
 Body2 {
  emitterSpawnScript [&FN_RibbonIDOffsetUtility()]
  emitterUpdateScript [
   &EmitterState()
   &FN_SpawnAnimTrail()
  ]
  particleSpawnScript [
   &FN_InitializeAnimTrailBoneIndecies(First Bone or Socket Index=4, Second Bone or Socket Index=5)
   &FN_InitializeAnimTrail(
    AnimTrailWidth=2
    Color=&MultiplyLinearColor(
     Linear Color=<rgba 0.1979 0.1119 0.0783 1>
     Scale Factor=&MakeVector4_FromLinearColor(LinearColor=$System.TODColor)
    )
    Lifetime=0.14
    SkeletalMesh=$System.Skeletal Mesh
   )
  ]
  particleUpdateScript [
   &FN_UpdateAnimTrail(SkeletalMesh=$System.Skeletal Mesh)
   &ScaleColor(
    Scale Alpha=&Multiply_Float(
     A=&FloatFromCurve(
      FloatCurve=DataInterface() {propertyValues {Curve {keys [<at 0 1.4 RCIM_Cubic>, {interpMode=RCIM_Cubic, leaveTangent=-2.522, tangentMode=RCTM_Break, time=0.0879, value=1}, <at 0.6 0>]}}}
      Scale Curve=$System.SwingGeneralFade
     )
     B=1
    )
    Scale RGB=<0.8 0.8 0.8>
    ScaleA=true
    ScaleRGB=true
   )
   &ScaleColor(Scale RGB=&VectorFromFloat(Value=$System.TODFloat), ScaleA=false, ScaleRGB=true) {module=ScaleColor001}
  ]
  rendererClasses [&NiagaraRibbonRendererProperties]
  renderers [{rendererClass=&NiagaraRibbonRendererProperties}]
 }
}
summary {
 systemName=NS_Pickaxe_GnocchiTea_Trail
 emitters [
  {emitterName=Body, rendererClasses [&NiagaraRibbonRendererProperties]}
  {emitterName=Blades, rendererClasses [&NiagaraRibbonRendererProperties]}
  {emitterName=Body2, rendererClasses [&NiagaraRibbonRendererProperties]}
 ]
 userVariables [
  User.bIsFrontEnd {type=NiagaraBool}
  User.bIsFrontEndPreview {type=NiagaraBool}
  User.RibbonIDOffset {type=NiagaraInt32}
 ]
}
