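# Control Rig for the GnocchiTea two-handed melee pickaxe that repositions its dynamic bones.
# The forward graph runs a single Modify Transforms unit at full weight from Begin Execution, with no variables or other logic.
# It applies a constant translation to a list of bones including dyn_I and dyn_D, which belong to five bones arranged radially around dyn_shild under the handle.
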ControlRigBlueprint /BRCosmetics/Weapons/FORT_Melee/Pickaxe_GnocchiTea/Meshes/GnocchiTea_Axe_CtrlRig
graphs [
 RigVMFunctionLibrary {links [], nodes []}
 RigVMModel {nodes [
   RigUnit_BeginExecution = &RigUnit_BeginExecution()
   &RigUnit_ModifyTransforms(
    ItemToModify="((Item=(Type=Bone,Name=\"dyn_I\"),Transform=(Rotation=(X=0,Y=0,Z=0,W=1),Translation=(X=-19,Y=-0.5,Z=0),Scale3D=(X=1,Y=1,Z=1))),(Item=(Type=Bone,Name=\"dyn_D\"),Transform=(Rotation=(X=0,Y=0,Z=0,W=1),Translation=(X=-19,Y=-0.5,Z=0),Scale3D=(X=1,Y=1,Z=1))),(Item=( ...(truncated)"
    Weight=1
    WeightMinimum=0
    ExecutePin=@RigUnit_BeginExecution.ExecutePin
   )
  ]}
]
hierarchy [
 b_AR_Root {loc=<0 0 0>, rot=<0 0 0 1>}
 Melee_TwoHanded_Handle {parent=b_AR_Root, loc=<0 0 0>, rot=<0 0 0.7071 -0.7071>}
 dyn_shild {parent=Melee_TwoHanded_Handle, loc=<6.4965 2.5 0>, rot=<0 0 0 1>}
 dyn_I {parent=dyn_shild, loc=<-0.0026 3.2886 12.5768>, rot=<-0.4322 0.4322 -0.5596 -0.5596>}
 dyn_D {parent=dyn_shild, loc=<-0.0026 12.9775 0.7588>, rot=<-0.0205 0.0205 -0.7068 -0.7068>}
 dyn_C {parent=dyn_shild, loc=<-0.0026 4.7319 -12.1078>, rot=<0.3988 -0.3988 -0.5839 -0.5839>}
 dyn_B {parent=dyn_shild, loc=<-0.0026 -10.053 -8.2419>, rot=<0.6658 -0.6658 -0.238 -0.238>}
 dyn_A {parent=dyn_shild, loc=<-0.0026 -10.945 7.0141>, rot=<0.6786 -0.6786 0.1988 0.1988>}
]
