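# Animation Blueprint for the GnocchiTea pickaxe, spinning its dyn_shild blade bone during swings and special actions.
# Each update detects melee via the owner's MeleeAnim curve or harvesting montage; this and special actions set a delayed spin flag.
# The flag snaps rotation speed high; otherwise speed eases to zero, faster in the front end.
# The accumulated wrapped angle drives additive bone-space roll after the slot and Control Rig pass.
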AnimBlueprint /BRCosmetics/Weapons/FORT_Melee/Pickaxe_GnocchiTea/Meshes/GnocchiTea_Axe_AnimBP
target_skeleton=&/GnocchiTea_Axe_Skeleton
graphs [
 AnimGraph {nodes [
   "Transform (Modify) Bone - Bone: dyn_shild"!(
    ComponentPose
    Alpha=1
    Translation="0,0,0"
    Rotation_Roll:float=Get!() -> (rotation of saw blade_Roll:float, rotation of saw blade_Pitch:float, rotation of saw blade_Yaw:float)
    Rotation_Pitch:float=0
    Rotation_Yaw:float=0
    Scale="1,1,1"
   ) -> (Pose) {
    rotationMode=BMM_Additive
    rotationSpace=BCS_BoneSpace
    scaleSpace=BCS_WorldSpace
    alphaBoolBlend {blendOption=Linear}
    boneToModify {boneName=dyn_shild}
   }
   AnimGraphNode_SubInput_0 = Input Pose()
   n1 = Control Rig!(Source=Slot 'DefaultSlot'(Source=@AnimGraphNode_SubInput_0), bAlphaBoolEnabled:bool=True) -> (Pose) {
    alphaInputType=Bool
    alphaBoolBlend {blendOption=Cubic, blendOutTime=0.6}
    controlRigAssetReference {blueprintRigClass=&/GnocchiTea_Axe_CtrlRig.GnocchiTea_Axe_CtrlRig_C}
   }
   n2 = "Transform (Modify) Bone - Bone: dyn_shild"!(
    ComponentPose=Local To Component(LocalPose=@n1)
    Alpha=1
    Translation="0,0,0"
    Rotation:Rotator=Get Rotator() -> (Rotator:Rotator)
    Scale="1,1,1"
   ) -> (Pose) {
    rotationMode=BMM_Additive
    rotationSpace=BCS_BoneSpace
    translationMode=BMM_Additive
    translationSpace=BCS_BoneSpace
    alphaBoolBlend {blendOption=Linear}
    alphaScaleBiasClamp {interpSpeedDecreasing=2, interpSpeedIncreasing=2}
    boneToModify {boneName=dyn_shild}
   }
   Output Pose(Result=Component To Local(ComponentPose=@n2)) {layerGroup=None}
  ]}
 EventGraph {nodes [
   n1 = Event BlueprintUpdateAnimation()
   > Set Delta Time X(Delta Time X=@n1.DeltaTimeX)
   > n2 = Sequence!(execute) -> (then_0, then_1, then_2:Exec, then_3:Exec, then_4:Exec)
   > n3 = Cast To FortWeaponPickaxeAthena!(execute, Object=GetOwningActor()) -> (then, CastFailed, AsFort Weapon Pickaxe Athena:Fort Weapon Pickaxe Athena Object Reference) {is=title K2Node_DynamicCast}
   n4 = GetOwner(self:Actor Object Reference=@n3)
   > Is Valid(InputObject=@n4)
   n5 = GetComponentByClass!(self:Actor Object Reference=@n4, ComponentClass:Actor Component Class Reference=/Script/Engine.SkeletalMeshComponent) -> (ReturnValue:Skeletal Mesh Component Object Reference)
   > Is Valid(InputObject=@n5)
   > n6 = Cast To FortPlayerAnimInstance(Object=GetAnimInstance(self=@n5))
   > Set Melee Anim Playing(Melee Anim Playing:bool="float >= float"(A=GetCurveValue(self=@n6, CurveName=MeleeAnim), B=1) {is=title K2Node_CallFunction}) -> (Output_Get:bool)
   n7 = GetAnimInstance(self=GetOwnerSkeletalMeshComponent())
   Set is Animation player melee combo(
    execute=@n2.then_4
    is Animation player melee combo:bool="Equal (Object)"(
     A=GetCurrentActiveMontage!(self:Anim Instance Object Reference=@n7) -> (ReturnValue:Anim Montage Object Reference)
     B=/Game/Animation/Game/MainPlayer/Combat/Melee/OneHanded/CAS/CAS_Pickaxe_Harvesting_Combo_CMM_M
    ) {is=title K2Node_CallFunction}
   ) -> (Output_Get:bool)
   n8 = Branch(execute=@n2.then_2, Condition=Get animation plus delay() -> (animation plus delay:bool))
   > Set RotationSpeed(RotationSpeed:double=1400) -> (Output_Get:double)
   Set RotationSpeed(
    execute=@n8.else
    RotationSpeed=FInterpTo(
     Current=Get RotationSpeed() -> (RotationSpeed:double)
     Target=0
     DeltaTime=GetWorldDeltaSeconds!() -> (ReturnValue:double)
     InterpSpeed=Get InterpSpeed() -> (InterpSpeed:double)
    )
   )
   n9 = float * float(A=Get Delta Time X(), B=Get RotationSpeed()) {is=title K2Node_PromotableOperator}
   n10 = Get current_velocity() -> (current_velocity:double)
   n11 = Set current_velocity(execute=@n2.then_3, current_velocity:double="% (Float)"!(A:double=float + float(A=@n9, B=@n10), B:double=360) -> (ReturnValue:double)) -> (Output_Get:double)
   > Set Rotator(Rotator:Rotator=MakeRotator(Roll=@n11)) -> (Output_Get:Rotator)
   n12 = Branch(execute=@n2.then_1, Condition=Get bIsInFrontEnd())
   > n13 = Set InterpSpeed(InterpSpeed:double=10) -> (Output_Get:double)
   n14 = Branch(
    execute=[@n13.then, @n15.then]
    Condition=OR Boolean(A=Get Melee Anim Playing() -> (Melee Anim Playing:bool), B=Get is Animation player melee combo() -> (is Animation player melee combo:bool))
   )
   > Delay(Duration=0.1)
   > Set animation plus delay(animation plus delay:bool=true) -> (Output_Get:bool)
   Set animation plus delay(execute=@n14.else, animation plus delay=false)
   n15 = Set InterpSpeed(execute=@n12.else, InterpSpeed=5)
   n16 = Event OnCharacterCustomizationLoaded()
   > Set GnocchiTea(GnocchiTea:Skeletal Mesh Component Object Reference=@n16.CustomBaseMesh) -> (Output_Get:Skeletal Mesh Component Object Reference)
   Event OnWeaponChanged()
   > n17 = Cast To FortPlayerAnimInstance(Object=GetAnimInstance(self=GetOwnerSkeletalMeshComponent()))
   > n18 = Switch on EFortWeaponCoreAnimation!(execute, Selection:EFortWeaponCoreAnimation Enum=Get WeaponCoreAnim(self=@n17)) -> (
    Melee:Exec
    Pistol:Exec
    Shotgun:Exec
    PaperBlueprint:Exec
    Rifle:Exec
    MeleeOneHand:Exec
    MachinePistol:Exec
    RocketLauncher:Exec
    GrenadeLauncher:Exec
    GoingCommando:Exec
    AssaultRifle:Exec
    TacticalShotgun:Exec
    SniperRifle:Exec
    TrapPlacement:Exec
    ShoulderLauncher:Exec
    AbilityDecoTool:Exec
    Crossbow:Exec
    C4:Exec
    RemoteControl:Exec
    DualWield:Exec
    AR_BullPup:Exec
    AR_ForwardGrip:Exec
    MedPackPaddles:Exec
    SMG_P90:Exec
    AR_DrumGun:Exec
    Consumable_Small:Exec
    Consumable_Large:Exec
    Balloon:Exec
    MountedTurret:Exec
    CreativeTool:Exec
    ExplosiveBow:Exec
    AshtonIndigo:Exec
    AshtonChicago:Exec
    MeleeDualWield:Exec
    Unarmed:Exec
   )
   n19 = Set move blades(execute=[@n18.AssaultRifle, @n18.PaperBlueprint, @n18.Pistol, @n18.Rifle, @n18.Shotgun], move blades:bool=true) -> (Output_Get:bool)
   Delay(execute=[@n19.then, @n20.then], Duration=0)
   > Set move blades(move blades=false)
   n20 = Set move blades(execute=[@n18.Melee, @n18.MeleeOneHand], move blades=true)
   n21 = Event OnPickaxeSpecialAction!() -> (OutputDelegate, then, ActionName:Name=None) {is=title K2Node_Event}
   > Delay(Duration=0.2)
   > Call On Sepcial Action Fired!(execute:Exec, self:Gnocchi Tea Axe Anim BP Object Reference, ActionName:Name=@n21.ActionName) -> (then:Exec) {is=title K2Node_CallDelegate}
   > Set animation plus delay(animation plus delay=true)
   > Delay(Duration=1)
   > Set animation plus delay(animation plus delay=false)
  ]}
 OnSepcialActionFired {nodes [OnSepcialActionFired!() -> (then, ActionName:Name) {is=title K2Node_FunctionEntry}]}
]
variables [
 470 {type=number}
 Alpha_IS_FrontEnd {type=boolean}
 Alpha_IS_PlayingMelee {type=boolean}
 Alpha_IS_PlayingMelee_02 {type=number}
 animation plus delay {type=boolean}
 current_velocity {type=number}
 CurrentAngle {type=number}
 currentgnocctea {type=AthenaBackpackItemDefinition, default=null}
 Delta Time X {type=number}
 GnocchiTea {type=SkeletalMeshComponent}
 Gravity Override Params {type=GravityOverrideParamsStruct}
 GravityOverride {type=Vector}
 InterpSpeed {type=number}
 is Animation player melee combo {type=boolean}
 isMelee {type=boolean}
 isPistol {type=boolean}
 Melee Anim Playing {type=boolean}
 move blades {type=boolean}
 NewVar {type=SkeletalMeshComponent}
 OnSepcialActionFired {default=()}
 Pickaxe {type=SkeletalMeshComponent}
 rotation of saw blade {type=Rotator}
 RotationSpeed {type=number}
 Rotator {type=Rotator}
]
