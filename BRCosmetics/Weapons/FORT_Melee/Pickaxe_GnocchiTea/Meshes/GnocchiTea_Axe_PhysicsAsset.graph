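# Physics asset for the GnocchiTea axe, a two-handed melee pickaxe cosmetic.
# A large capsule covers the Melee_TwoHanded_Handle bone, while a tiny capsule gives the dyn_shild bone its own body.
# A single constraint joins dyn_shild to the handle with both swing axes locked, so the part can only rotate around its twist axis.
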PhysicsAsset /BRCosmetics/Weapons/FORT_Melee/Pickaxe_GnocchiTea/Meshes/GnocchiTea_Axe_PhysicsAsset
bodies {
 Melee_TwoHanded_Handle [{
   shapeType=Capsule
   length=28.714
   radius=27.9178
   shapeName=Melee_TwoHanded_Handle_capsule
   center=<9.4497 2.7979 2.2685>
   rotation=<rot 87.3271 -90.3996 -0.3991>
  }]
 dyn_shild [{shapeType=Capsule, length=0.505, radius=0.505, shapeName=dyn_shild_capsule, rotation=<rot 0 -90 0>}]
}
constraints [{
  bone1Name=dyn_shild
  bone2Name=Melee_TwoHanded_Handle
  swing1LimitDegrees=45
  swing1Motion=Locked
  swing2LimitDegrees=45
  swing2Motion=Locked
  twistLimitDegrees=180
 }]
